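# Sound cue for the SodaMug pickaxe hitting an enemy, using pickaxe impact attenuation and property sheet.
# After a short fixed delay, a weighted random node picks either a base impact alone or, slightly less often, a base impact mixed with a fireworks layer.
# Each branch draws from six wave variations per layer, and only the fireworks layer gets slight pitch and volume modulation.
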SoundCue /BRCosmetics/Sounds/PickAxes/SodaMug/PickaxeImpactEnemy_SodaMug_Cue
root=Delay_0
nodes [
 {id=Delay_0, children [Random_5], properties {delayMax=0.1, delayMin=0.1}}
 {id=Mixer_1, children [Random_4,Modulator_0], properties {inputVolume [1,1]}}
 {id=Modulator_0, children [Random_3], properties {pitchMax=1.02, pitchMin=0.98, volumeMax=1.02, volumeMin=0.98}}
 {
  id=Random_1
  children [WavePlayer_6, WavePlayer_7, WavePlayer_8, WavePlayer_9, WavePlayer_10, WavePlayer_11]
  properties {weights [1,1,1,1,1,1]}
 }
 {
  id=Random_3
  children [WavePlayer_12, WavePlayer_13, WavePlayer_14, WavePlayer_15, WavePlayer_16, WavePlayer_17]
  properties {weights [1,1,1,1,1,1]}
 }
 {
  id=Random_4
  children [WavePlayer_18, WavePlayer_19, WavePlayer_20, WavePlayer_21, WavePlayer_22, WavePlayer_23]
  properties {weights [1,1,1,1,1,1]}
 }
 {id=Random_5, children [Random_1,Mixer_1], properties {bRandomizeWithoutReplacement=false, weights [1,0.8]}}
 {id=WavePlayer_10, properties {soundWaveAssetPtr=&/Pickaxe_SodaMug_Impact_Base_05}}
 {id=WavePlayer_11, properties {soundWaveAssetPtr=&/Pickaxe_SodaMug_Impact_Base_06}}
 {id=WavePlayer_12, properties {soundWaveAssetPtr=&/Pickaxe_SodaMug_Impact_FireworksLayer_06}}
 {id=WavePlayer_13, properties {soundWaveAssetPtr=&/Pickaxe_SodaMug_Impact_FireworksLayer_05}}
 {id=WavePlayer_14, properties {soundWaveAssetPtr=&/Pickaxe_SodaMug_Impact_FireworksLayer_04}}
 {id=WavePlayer_15, properties {soundWaveAssetPtr=&/Pickaxe_SodaMug_Impact_FireworksLayer_03}}
 {id=WavePlayer_16, properties {soundWaveAssetPtr=&/Pickaxe_SodaMug_Impact_FireworksLayer_02}}
 {id=WavePlayer_17, properties {soundWaveAssetPtr=&/Pickaxe_SodaMug_Impact_FireworksLayer_01}}
 {id=WavePlayer_18, properties {soundWaveAssetPtr=&/Pickaxe_SodaMug_Impact_Base_01}}
 {id=WavePlayer_19, properties {soundWaveAssetPtr=&/Pickaxe_SodaMug_Impact_Base_02}}
 {id=WavePlayer_20, properties {soundWaveAssetPtr=&/Pickaxe_SodaMug_Impact_Base_03}}
 {id=WavePlayer_21, properties {soundWaveAssetPtr=&/Pickaxe_SodaMug_Impact_Base_04}}
 {id=WavePlayer_22, properties {soundWaveAssetPtr=&Pickaxe_SodaMug_Impact_Base_05}}
 {id=WavePlayer_23, properties {soundWaveAssetPtr=&Pickaxe_SodaMug_Impact_Base_06}}
 {id=WavePlayer_6, properties {soundWaveAssetPtr=&Pickaxe_SodaMug_Impact_Base_01}}
 {id=WavePlayer_7, properties {soundWaveAssetPtr=&Pickaxe_SodaMug_Impact_Base_02}}
 {id=WavePlayer_8, properties {soundWaveAssetPtr=&Pickaxe_SodaMug_Impact_Base_03}}
 {id=WavePlayer_9, properties {soundWaveAssetPtr=&Pickaxe_SodaMug_Impact_Base_04}}
]
properties {
 attenuationSettings=/Game/Sounds/Attenuation/Fort_Pickaxe_Impact_Att
 audioPropertiesSheet=/BRCosmetics/Sounds/Settings/PropertySheets/Pickaxe/APS_Cosmetic_Pickaxe_Impact
 maxDistance=4300
 soundClassObject=&Fort_Impacts
 concurrencySet [null]
}
