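# Sound cue for readying the SodaMug cosmetic pickaxe, mixing a close and a medium attenuation layer.
# The close layer blends a random equip sound with a slightly louder random sparkle that starts after a short randomized delay.
# The medium layer plays a random distant fire-axe swing borrowed from Athena weapon sounds.
# Equip and swing layers get slight pitch and volume variation, and the cue uses the cosmetic pickaxe-ready property sheet.
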SoundCue /BRCosmetics/Sounds/PickAxes/SodaMug/PickaxeReady_SodaMug_Cue
root=Mixer_0
nodes [
 {id=Attenuation_0, children [Mixer_1], properties {attenuationSettings=/Game/Sounds/Attenuation/Fort_Weapon_Pickaxe_Close_Att}}
 {id=Attenuation_1, children [Modulator_1], properties {attenuationSettings=/Game/Sounds/Attenuation/Fort_Weapon_Pickaxe_Medium_Att}}
 {id=Delay_0, children [Modulator_2], properties {delayMax=0.2, delayMin=0.15}}
 {id=Mixer_0, children [Attenuation_0, Attenuation_1], properties {inputVolume [1,1]}}
 {id=Mixer_1, children [Modulator_0,Delay_0], properties {inputVolume [1,1.25]}}
 {id=Modulator_0, children [Random_2], properties {pitchMax=1.02, pitchMin=0.98, volumeMax=1.02, volumeMin=0.98}}
 {id=Modulator_1, children [Random_1], properties {pitchMax=1.02, pitchMin=0.98, volumeMax=1.02, volumeMin=0.98}}
 {id=Modulator_2, children [Random_3]}
 {
  id=Random_1
  children [WavePlayer_6, WavePlayer_7, WavePlayer_8, WavePlayer_9, WavePlayer_10, WavePlayer_11]
  properties {weights [1,1,1,1,1,1]}
 }
 {
  id=Random_2
  children [WavePlayer_12, WavePlayer_13, WavePlayer_14, WavePlayer_15, WavePlayer_16, WavePlayer_17]
  properties {weights [1,1,1,1,1,1]}
 }
 {
  id=Random_3
  children [WavePlayer_18, WavePlayer_19, WavePlayer_20, WavePlayer_21, WavePlayer_22, WavePlayer_23]
  properties {weights [1,1,1,1,1,1]}
 }
 {id=WavePlayer_10, properties {soundWaveAssetPtr=/Game/Athena/Sounds/Weapons/Fire_Axe_Swing_Distant_05}}
 {id=WavePlayer_11, properties {soundWaveAssetPtr=/Game/Athena/Sounds/Weapons/Fire_Axe_Swing_Distant_06}}
 {id=WavePlayer_12, properties {soundWaveAssetPtr=&/Pickaxe_SodaMug_Equip_01}}
 {id=WavePlayer_13, properties {soundWaveAssetPtr=&/Pickaxe_SodaMug_Equip_02}}
 {id=WavePlayer_14, properties {soundWaveAssetPtr=&/Pickaxe_SodaMug_Equip_03}}
 {id=WavePlayer_15, properties {soundWaveAssetPtr=&/Pickaxe_SodaMug_Equip_04}}
 {id=WavePlayer_16, properties {soundWaveAssetPtr=&/Pickaxe_SodaMug_Equip_05}}
 {id=WavePlayer_17, properties {soundWaveAssetPtr=&/Pickaxe_SodaMug_Equip_06}}
 {id=WavePlayer_18, properties {soundWaveAssetPtr=&/Pickaxe_SodaMug_Equip_Sparkle_01}}
 {id=WavePlayer_19, properties {soundWaveAssetPtr=&/Pickaxe_SodaMug_Equip_Sparkle_02}}
 {id=WavePlayer_20, properties {soundWaveAssetPtr=&/Pickaxe_SodaMug_Equip_Sparkle_03}}
 {id=WavePlayer_21, properties {soundWaveAssetPtr=&/Pickaxe_SodaMug_Equip_Sparkle_04}}
 {id=WavePlayer_22, properties {soundWaveAssetPtr=&/Pickaxe_SodaMug_Equip_Sparkle_05}}
 {id=WavePlayer_23, properties {soundWaveAssetPtr=&/Pickaxe_SodaMug_Equip_Sparkle_06}}
 {id=WavePlayer_6, properties {soundWaveAssetPtr=/Game/Athena/Sounds/Weapons/Fire_Axe_Swing_Distant_01}}
 {id=WavePlayer_7, properties {soundWaveAssetPtr=/Game/Athena/Sounds/Weapons/Fire_Axe_Swing_Distant_02}}
 {id=WavePlayer_8, properties {soundWaveAssetPtr=/Game/Athena/Sounds/Weapons/Fire_Axe_Swing_Distant_03}}
 {id=WavePlayer_9, properties {soundWaveAssetPtr=/Game/Athena/Sounds/Weapons/Fire_Axe_Swing_Distant_04}}
]
properties {
 audioPropertiesSheet=/BRCosmetics/Sounds/Settings/PropertySheets/Pickaxe/APS_Cosmetic_Pickaxe_Ready
 maxDistance=3000
 soundClassObject=&Fort_WeaponMelee_PSM
 volumeMultiplier=0.28
 concurrencySet [null]
}
